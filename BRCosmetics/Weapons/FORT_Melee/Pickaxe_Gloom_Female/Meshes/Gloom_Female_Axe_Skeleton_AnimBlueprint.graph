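# Animation Blueprint for the Gloom Female pickaxe skeleton that adds secondary physics jiggle to its umbrella bones.
# The input pose passes through a serial chain of Anim Dynamics nodes in component space: a linear-spring top bone, two inner bones, then six outer bones.
# Outer bones use stiff angular springs, with varied alpha, limits and damping so they move unevenly.
# The event graph holds only unconnected default nodes; there are no variables.
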AnimBlueprint /BRCosmetics/Weapons/FORT_Melee/Pickaxe_Gloom_Female/Meshes/Gloom_Female_Axe_Skeleton_AnimBlueprint
target_skeleton=&/Gloom_Female_Axe_Skeleton
graphs [
 AnimGraph {nodes [
   n1 = Anim Dynamics - Bone: dyn_umbrella_top(ComponentPose=Local To Component(LocalPose=Input Pose()), Alpha=1, ExternalForce="0,0,0") {
    angularDampingOverride=0.999
    bLinearSpring=true
    bOverrideAngularDamping=true
    bOverrideLinearDamping=true
    bUseGravityOverride=true
    linearDampingOverride=0.999
    linearSpringConstant=200
    alphaBoolBlend {blendOption=Linear}
    boundBone {boneName=dyn_umbrella_top}
    componentAppliedLinearAccClamp=<3000 3000 3000>
    componentLinearAccScale=<0.2 0.2 0.3>
    componentLinearVelScale=<0.2 0.2 0.3>
    gravityOverride {z=-400}
    physicsBodyDefinitions [{
      boundBone {boneName=dyn_umbrella_top}
      boxExtents=<30 15 15>
      constraintSetup {angularLimitsMax=<0 0 0>, angularLimitsMin=<0 0 0>, angularTarget=<0 0 0>, linearAxesMax=<0.1 2 3>, linearAxesMin=<-0.1 -3 -2>}
      localJointOffset=<1 0 0>
     }]
   }
   n2 = Anim Dynamics - Bone: dyn_umbrealla_inner_01(ComponentPose=@n1, Alpha=1, ExternalForce="0,0,0") {
    angularDampingOverride=0.999
    angularSpringConstant=100
    bOverrideAngularDamping=true
    bUseGravityOverride=true
    linearDampingOverride=0.999
    linearSpringConstant=100
    alphaBoolBlend {blendOption=Linear}
    boundBone {boneName=dyn_umbrealla_inner_01}
    componentAppliedLinearAccClamp=<3000 3000 3000>
    componentLinearAccScale=<0.2 0.2 0.35>
    componentLinearVelScale=<0.2 0.2 0.35>
    gravityOverride {z=-400}
    physicsBodyDefinitions [{
      boundBone {boneName=dyn_umbrealla_inner_01}
      boxExtents=<30 20 20>
      constraintSetup {angularLimitsMax=<1 10 10>, angularLimitsMin=<-1 -10 0>, angularTarget=<0 0 0>}
      localJointOffset=<5 0 0>
     }]
   }
   n3 = Anim Dynamics - Bone: dyn_umbrealla_inner_02(ComponentPose=@n2, Alpha=1, ExternalForce="0,0,0") {
    angularDampingOverride=0.999
    angularSpringConstant=100
    bOverrideAngularDamping=true
    bUseGravityOverride=true
    linearDampingOverride=0.999
    linearSpringConstant=100
    alphaBoolBlend {blendOption=Linear}
    boundBone {boneName=dyn_umbrealla_inner_02}
    componentAppliedLinearAccClamp=<3000 3000 3000>
    componentLinearAccScale=<0.2 0.2 0.35>
    componentLinearVelScale=<0.2 0.2 0.35>
    gravityOverride {z=-400}
    physicsBodyDefinitions [{
      boundBone {boneName=dyn_umbrealla_inner_02}
      boxExtents=<30 20 20>
      constraintSetup {angularLimitsMax=<1 10 10>, angularLimitsMin=<-1 -10 0>, angularTarget=<0 0 0>}
      localJointOffset=<5 0 0>
     }]
   }
   n4 = Anim Dynamics - Bone: dyn_umbrealla_outer_01(ComponentPose=@n3, Alpha=0.8, ExternalForce="0,0,0") {
    angularDampingOverride=0.992
    angularSpringConstant=18500
    bAngularSpring=true
    bOverrideAngularDamping=true
    bUseGravityOverride=true
    linearDampingOverride=0.999
    linearSpringConstant=100
    alphaBoolBlend {blendOption=Linear}
    boundBone {boneName=dyn_umbrealla_outer_01}
    componentAppliedLinearAccClamp=<3000 3000 3000>
    componentLinearAccScale=<0.23 0.23 0.33>
    componentLinearVelScale=<0.2 0.2 0.35>
    gravityOverride {z=-400}
    physicsBodyDefinitions [{
      boundBone {boneName=dyn_umbrealla_outer_01}
      boxExtents=<60 60 30>
      constraintSetup {angularLimitsMax=<2 12 20>, angularLimitsMin=<-2 -12 -10>, angularTarget=<1 0 0>}
      localJointOffset=<10 0 0>
     }]
   }
   n5 = Anim Dynamics - Bone: dyn_umbrealla_outer_02(ComponentPose=@n4, Alpha=0.6, ExternalForce="0,0,0") {
    angularDampingOverride=0.999
    angularSpringConstant=18500
    bAngularSpring=true
    bOverrideAngularDamping=true
    bUseGravityOverride=true
    linearDampingOverride=0.999
    linearSpringConstant=100
    alphaBoolBlend {blendOption=Linear}
    boundBone {boneName=dyn_umbrealla_outer_02}
    componentAppliedLinearAccClamp=<3000 3000 3000>
    componentLinearAccScale=<0.28 0.28 0.38>
    componentLinearVelScale=<0.2 0.2 0.35>
    gravityOverride {z=-400}
    physicsBodyDefinitions [{
      boundBone {boneName=dyn_umbrealla_outer_02}
      boxExtents=<60 60 30>
      constraintSetup {angularLimitsMax=<2 18 30>, angularLimitsMin=<-2 -18 -10>, angularTarget=<1 0 0>}
      localJointOffset=<9 0 0>
     }]
   }
   n6 = Anim Dynamics - Bone: dyn_umbrealla_outer_03(ComponentPose=@n5, Alpha=0.7, ExternalForce="0,0,0") {
    angularDampingOverride=0.998
    angularSpringConstant=18500
    bAngularSpring=true
    bOverrideAngularDamping=true
    bUseGravityOverride=true
    linearDampingOverride=0.999
    linearSpringConstant=100
    alphaBoolBlend {blendOption=Linear}
    boundBone {boneName=dyn_umbrealla_outer_03}
    componentAppliedLinearAccClamp=<3000 3000 3000>
    componentLinearAccScale=<0.25 0.25 0.32>
    componentLinearVelScale=<0.2 0.2 0.35>
    gravityOverride {z=-400}
    physicsBodyDefinitions [{
      boundBone {boneName=dyn_umbrealla_outer_03}
      boxExtents=<60 60 30>
      constraintSetup {angularLimitsMax=<2 10 15>, angularLimitsMin=<-2 -10 -10>, angularTarget=<1 0 0>}
      localJointOffset=<8 0 0>
     }]
   }
   n7 = Anim Dynamics - Bone: dyn_umbrealla_outer_04(ComponentPose=@n6, Alpha=0.6, ExternalForce="0,0,0") {
    angularDampingOverride=0.995
    angularSpringConstant=18500
    bAngularSpring=true
    bOverrideAngularDamping=true
    bUseGravityOverride=true
    linearDampingOverride=0.999
    linearSpringConstant=100
    alphaBoolBlend {blendOption=Linear}
    boundBone {boneName=dyn_umbrealla_outer_04}
    componentAppliedLinearAccClamp=<3000 3000 3000>
    componentLinearAccScale=<0.28 0.28 0.38>
    componentLinearVelScale=<0.2 0.2 0.35>
    gravityOverride {z=-400}
    physicsBodyDefinitions [{
      boundBone {boneName=dyn_umbrealla_outer_04}
      boxExtents=<60 60 30>
      constraintSetup {angularLimitsMax=<2 12 20>, angularLimitsMin=<-2 -12 -10>, angularTarget=<1 0 0>}
      localJointOffset=<10 0 0>
     }]
   }
   n8 = Anim Dynamics - Bone: dyn_umbrealla_outer_05(ComponentPose=@n7, Alpha=0.8, ExternalForce="0,0,0") {
    angularDampingOverride=0.992
    angularSpringConstant=18500
    bAngularSpring=true
    bOverrideAngularDamping=true
    bUseGravityOverride=true
    linearDampingOverride=0.999
    linearSpringConstant=100
    alphaBoolBlend {blendOption=Linear}
    boundBone {boneName=dyn_umbrealla_outer_05}
    componentAppliedLinearAccClamp=<3000 3000 3000>
    componentLinearAccScale=<0.25 0.25 0.35>
    componentLinearVelScale=<0.2 0.2 0.35>
    gravityOverride {z=-400}
    physicsBodyDefinitions [{
      boundBone {boneName=dyn_umbrealla_outer_05}
      boxExtents=<60 60 30>
      constraintSetup {angularLimitsMax=<2 7 23>, angularLimitsMin=<-2 -7 -10>, angularTarget=<1 0 0>}
      localJointOffset=<8 0 0>
     }]
   }
   n9 = Anim Dynamics - Bone: dyn_umbrealla_outer_06(ComponentPose=@n8, Alpha=0.7, ExternalForce="0,0,0") {
    angularDampingOverride=0.998
    angularSpringConstant=18500
    bAngularSpring=true
    bOverrideAngularDamping=true
    bUseGravityOverride=true
    linearDampingOverride=0.999
    linearSpringConstant=100
    alphaBoolBlend {blendOption=Linear}
    boundBone {boneName=dyn_umbrealla_outer_06}
    componentAppliedLinearAccClamp=<3000 3000 3000>
    componentLinearAccScale=<0.25 0.25 0.35>
    componentLinearVelScale=<0.2 0.2 0.35>
    gravityOverride {z=-400}
    physicsBodyDefinitions [{
      boundBone {boneName=dyn_umbrealla_outer_06}
      boxExtents=<60 60 30>
      constraintSetup {angularLimitsMax=<2 15 17>, angularLimitsMin=<-2 -15 -10>, angularTarget=<1 0 0>}
      localJointOffset=<12 0 0>
     }]
   }
   Output Pose(Result=Component To Local(ComponentPose=@n9)) {layerGroup=None}
  ]}
 EventGraph {nodes [
   TryGetPawnOwner()
   Event BlueprintUpdateAnimation()
  ]}
]
variables []
